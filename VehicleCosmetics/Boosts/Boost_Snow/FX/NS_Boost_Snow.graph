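# Niagara system for the Snow vehicle boost cosmetic, trailing snow sprites from the car's boost sockets.
# Three sprite emitters layer snow puffs, GPU-simulated tumbling flakes and short-lived fast dust.
# Each partitions particles across filtered skeletal mesh sockets, launches them in cones and inherits source movement.
# Painted-variant color overrides set spawn tint, while frontend-preview and painted user flags switch values such as speed, size, brightness and alpha.
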NiagaraSystem /VehicleCosmetics/Boosts/Boost_Snow/FX/NS_Boost_Snow
dependencies {
 usedDataInterfaces [
  &NiagaraDataInterfaceCurve
  &NiagaraDataInterfaceArrayFloat
  &NiagaraDataInterfaceSpriteRendererInfo
  &NiagaraDataInterfaceVector2DCurve
  &NiagaraDataInterfaceVectorCurve
  &NiagaraDataInterfaceSkeletalMesh
 ]
 usedDynamicInputs [/VehicleCosmetics/FX/Niagara/Modules/MergeEmitters/VC_FloatBySpawnGroup, &MakeFloatFromVector2d]
 usedModules [
  &self:SystemSpawnScript.SystemScriptSource.SystemScriptGraph.NiagaraNodeAssignment_0.SetVariables_#1
  &SystemState
 ]
}
emitters {
 NE_VC_Boost_Sprite_DUST {
  emitterUpdateScript [
   &EmitterState()
   &VC_FilteredSocketCount()
   &SpawnRate(SpawnRate=200, Use Spawn Probability=false)
  ]
  particleSpawnScript [
   &VC_PaintedVariantsOverride(
    BurntSienna=<rgba 0.4 0.1235 0.0493 1>
    Cobalt=<rgba 0.1 0.1166 1 2>
    Crimson=<rgba 0.7 0.021 0.021 1>
    ForestGreen=<rgba 0.0233 0.7 0 1>
    Grey=<rgba 0.15 0.15 0.15 1>
    Lime=<rgba 0.35 0.7 0 1>
    Pink=<rgba 1 0.3 0.7455 1>
    Purple=<rgba 0.4104 0 1 1>
    Saffron=<rgba 0.9 0.9 0 1>
    TitaniumWhite=<rgba 1 1 1 1>
   )
   &InitializeParticle(
    Color=$Particles.Color
    Lifetime=&RandomRangeFloat(Minimum=0.2, Maximum=0.25)
    Position Offset=&AddVectorToPosition(Position=<0 0 0>, Vector=<0 0 0>)
    Sprite Size Max=<1 4>
    Sprite Size Min=<0.75 7>
    Sprite Size Mode=::NewEnumerator2
    UsePositionOffset=false
    Write Material Random=true
   )
   &PartitionParticles(Number Of Partitions=$Emitter.PortCount)
   &SkeletalMeshLocation(
    Bone / Socket Index=$Particles.PartitionParticles.Partition
    Bone / Socket Sampling Mode=::NewEnumerator5
    MaskSpawnGroup=false
    Mesh Sampling Type=::NewEnumerator6
    Sampled Position Offset=&AddVectorToPosition(Position=<-25 0 0>, Vector=$User.DTCarOffset)
    Skeletal Mesh=$System.SkeletalMesh
   )
   &AddVelocity(
    Cone Angle=40
    Cone Axis=<-1 0 0>
    Distribution Along Cone Axis=100
    Inner Cone Angle=10
    Module.UseConeFalloff=false
    Module.UseVelDistribution=true
    Override Seed=false
    Velocity Mode=::NewEnumerator2
    Velocity Speed=3000
   )
   &ShapeLocation(MaskSpawnGroup=false, Non Uniform Scale=<1 1 2>, Override Seed=false, Sphere Radius=20, UseOwnerScale=true) {module=ShapeLocation001}
  ]
  particleUpdateScript [
   &ParticleState()
   &ScaleSpriteSize(
    Non-Uniform Curve Scale=&RandomRangeVector2D(Minimum=<2 2.5>, Maximum=<3 2>)
    Non-Uniform Curve Sprite Scale=DataInterface() {propertyValues {
      XCurve {keys [{tangentMode=RCTM_User, time=0, value=2}, {tangentMode=RCTM_User, time=1, value=2}]}
      YCurve {keys [{tangentMode=RCTM_User, time=0, value=4}, {tangentMode=RCTM_User, time=1, value=1.5}]}
     }}
    Scale Sprite Size Mode=::NewEnumerator3
   )
   &ScaleVelocity(
    Coordinate Space:ENiagaraCoordinateSpace=::Simulation
    Velocity Scale=&VectorFromCurve(VectorCurve=DataInterface() {propertyValues {
       XCurve {keys [<at 0 0.2 RCIM_Cubic>,<at 0.5 1 RCIM_Cubic>]}
       YCurve {keys [<at 0 0.2 RCIM_Cubic>,<at 0.5 1 RCIM_Cubic>]}
       ZCurve {keys [<at 0 0.2 RCIM_Cubic>,<at 0.5 1 RCIM_Cubic>]}
      }})
   )
   &Drag(
    Drag=&FloatFromCurve(FloatCurve=DataInterface() {propertyValues {Curve {keys [<at 0 0 RCIM_Cubic>,<at 0.7 1 RCIM_Cubic>]}}}, Scale Curve=3)
    Use Linear Drag=true
    Use Rotational Drag=true
   )
   &InheritSourceMovement(
    Applied Position Scale=&Multiply_VectorByFloat(Vector=<1 1 1>, Float=1)
    ApplyForce=false
    ApplyPosition=true
    ApplyPositionFalloff=false
   )
   &SolveForcesAndVelocity(Clamp Velocity=false, Limit Acceleration=false)
   &ScaleColor(
    Scale RGB=&VectorFromFloat(Value=&MakeCustomFloatFromBool(Bool=$User.bIsFrontendPreview, True Float=10, False Float=20))
    ScaleA=true
    ScaleRGB=true
   )
  ]
  rendererClasses [&NiagaraSpriteRendererProperties]
  renderers [{rendererClass=&NiagaraSpriteRendererProperties}]
 }
 NE_VC_Boost_Sprite_FLAKES {
  simTarget=GPUComputeSim
  emitterUpdateScript [
   &EmitterState()
   &VC_FilteredSocketCount()
   &SpawnRate(SpawnRate=150, Use Spawn Probability=false)
  ]
  particleSpawnScript [
   &VC_PaintedVariantsOverride(
    Black=<rgba 0 0 0 3>
    BurntSienna=<rgba 0.7 0.3267 0 0.99>
    Cobalt=<rgba 0.4 0.6667 2 0.9>
    Crimson=<rgba 1 0 0 1>
    ForestGreen=<rgba 0 1 0 1>
    Grey=<rgba 0.2 0.2 0.2 1.5>
    Lime=<rgba 0.6667 1 0 1>
    Orange=<rgba 1 0.5 0 1>
    Pink=<rgba 1 0.35 0.7636 1>
    Purple=<rgba 0.5833 0 1 1.1>
    Saffron=<rgba 1 1 0 1>
    SkyBlue=<rgba 0.12 0.84 1.2 1.1>
    TitaniumWhite=<rgba 1 1 1 1>
   )
   &InitializeParticle(
    Color=$Particles.Color
    Lifetime Max=0.6
    Lifetime Min=0.4
    Lifetime Mode=::NewEnumerator1
    Position Offset=&AddVectorToPosition(Position=<0 0 0>, Vector=<0 0 0>)
    Sprite Rotation Mode=::NewEnumerator2
    Sprite Size Mode=::NewEnumerator1
    Uniform Sprite Size Max=35
    Uniform Sprite Size Min=&SelectFloatFromArray(
     Array Sampling Mode=::NewEnumerator0
     Direct Array Index=?
     Float Selection Array=DataInterface() {propertyValues {FloatData [1,1.5]}}
     Random Seed=0
    )
    UsePositionOffset=false
    Write Material Random=true
   )
   &PartitionParticles(Number Of Partitions=$Emitter.PortCount)
   &SkeletalMeshLocation(
    Bone / Socket Index=$Particles.PartitionParticles.Partition
    Bone / Socket Sampling Mode=::NewEnumerator5
    MaskSpawnGroup=false
    Mesh Sampling Type=::NewEnumerator6
    Sampled Position Offset=&AddVectorToPosition(Position=<-35 0 0>, Vector=$User.DTCarOffset)
    Skeletal Mesh=$System.SkeletalMesh
   )
   &ApplyInitialForces()
   &AddVelocity(
    Cone Angle=15
    Cone Axis=&MakeCustomVectorFromBool(
     Bool=$User.bIsFrontendPreview
     True Vector=<1 0 -0.05>
     False Vector=&RandomRangeVector(Minimum=<1 -0.2 -0.2>, Maximum=<1 0.2 -0.2>)
    )
    Inner Cone Angle=15
    Module.UseConeFalloff=false
    Module.UseVelDistribution=false
    Override Seed=false
    Velocity Mode=::NewEnumerator2
    Velocity Speed=&MakeCustomFloatFromBool(Bool=$User.bIsFrontendPreview, True Float=-3000, False Float=-1500)
   )
   &ShapeLocation(MaskSpawnGroup=false, Override Seed=false, Sphere Radius=70, UseOwnerScale=true)
  ]
  particleUpdateScript [
   &ParticleState()
   &ScaleSpriteSize(
    Uniform Curve Scale=&MakeCustomFloatFromBool(Bool=$User.bIsFrontendPreview, True Float=1, False Float=1.5)
    Uniform Curve Sprite Scale=DataInterface() {propertyValues {Curve {keys [<at 0 1 RCIM_Cubic>, {arriveTangent=2.7607, interpMode=RCIM_Cubic, leaveTangent=2.7607, tangentMode=RCTM_User, time=1, value=2}]}}}
   )
   &SpriteRotationRate(Rotation Rate=&RandomRangeFloat(Minimum=-800, Maximum=800))
   &Drag(
    Drag=&FloatFromCurve(
     FloatCurve=DataInterface() {propertyValues {Curve {keys [
         {arriveTangent=0.0201, interpMode=RCIM_Cubic, leaveTangent=0.0201, tangentMode=RCTM_User, time=0, value=0}
         {arriveTangent=0.0695, interpMode=RCIM_Cubic, leaveTangent=0.0695, tangentMode=RCTM_User, time=0.9, value=1}
        ]}}}
     Scale Curve=3
    )
    Use Linear Drag=true
    Use Rotational Drag=true
   )
   &InheritSourceMovement(
    Applied Position Scale=&Multiply_VectorByFloat(Vector=<1 1 1>, Float=1)
    ApplyForce=false
    ApplyPosition=true
    ApplyPositionFalloff=false
   )
   &SolveForcesAndVelocity(Clamp Velocity=false, Limit Acceleration=false)
   &DynamicMaterialParameters(
    Index 0 Param 1=&MakeCustomFloatFromBool(
     Bool=$User.bIsPainted
     True Float=&MakeCustomFloatFromBool(Bool=$User.bIsFrontendPreview, True Float=$User.FLAKESBRIGHTS, False Float=1)
     False Float=&MakeCustomFloatFromBool(Bool=$User.bIsFrontendPreview, True Float=&RandomRangeFloat(Minimum=1.5, Maximum=0.8), False Float=2)
    )
    Module.Index0Param0Write=true
    Module.Index0Param1Write=true
    Module.Index0Param2Write=true
    Module.Index0Param3Write=true
   )
   &JitterPosition(Jitter Amount=10, Jitter Delay=0.1)
   &ScaleColor(
    Scale Alpha=&FloatFromCurve(
     FloatCurve=DataInterface() {propertyValues {Curve {keys [
         <at 0 0.25 RCIM_Cubic>
         {arriveTangent=-0.25, interpMode=RCIM_Cubic, leaveTangent=-0.25, time=0.1, value=1}
         {arriveTangent=-3.1038, interpMode=RCIM_Cubic, leaveTangent=-3.1038, tangentMode=RCTM_User, time=1, value=0}
        ]}}}
     Scale Curve=&MakeCustomFloatFromBool(
      Bool=$User.bIsFrontendPreview
      True Float=&RandomRangeFloat(Minimum=0.3, Maximum=0.5)
      False Float=&MakeCustomFloatFromBool(Bool=$User.bIsPainted, True Float=1, False Float=&RandomRangeFloat(Minimum=0.5, Maximum=0.6))
     )
    )
    Scale RGB=&VectorFromFloat(Value=&FloatFromCurve(FloatCurve=DataInterface() {propertyValues {Curve {keys [<at 0 1 RCIM_Cubic>,<at 1 0 RCIM_Cubic>]}}}, Scale Curve=2))
    ScaleA=true
    ScaleRGB=true
   )
   &SubUVAnimation(
    Evaluation Type=::NewEnumerator0
    "Sprite Renderer "=DataInterface() {propertyValues {SpriteRenderer=&self:NE_VC_Boost_Sprite_FLAKES_0.NiagaraSpriteRendererProperties_3}}
    SubUV Animation Mode:ENiagara_SubUVLookupModeV2=::NewEnumerator2
    UseEndFrame=false
    UseStartFrame=false
   )
  ]
  rendererClasses [&NiagaraSpriteRendererProperties]
  renderers [{rendererClass=&NiagaraSpriteRendererProperties}]
 }
 NE_VC_Boost_Sprite_Snow {
  emitterUpdateScript [
   &EmitterState()
   &VC_FilteredSocketCount()
   &SpawnRate(
    SpawnRate=&MakeCustomFloatFromBool(Bool=$User.bIsFrontendPreview, True Float=80, False Float=110)
    Use Spawn Probability=false
   )
  ]
  particleSpawnScript [
   &VC_PaintedVariantsOverride(
    Black=<rgba 0.03 0.03 0.03 2.5>
    BurntSienna=<rgba 0.25 0.0521 0 0.85>
    Cobalt=<rgba 0.075 0.165 0.75 0.8>
    Crimson=<rgba 0.6 0.0065 0.006 0.75>
    DefaultColor=&MakeCustomLinearColorFromBool(Bool=$User.bIsFrontendPreview, True Color=<rgba 0.54 0.903 1.2 0.8>, False Color=<rgba 0.45 0.7525 1 1>)
    ForestGreen=<rgba 0 0.5 0 0.85>
    Grey=<rgba 0.3 0.3 0.3 0.7>
    Lime=<rgba 0.5545 1 0.01 0.8>
    Orange=<rgba 0.85 0.3088 0 0.8>
    Pink=<rgba 0.8 0.24 0.7533 0.7>
    Purple=<rgba 0.23 0 0.6 0.65>
    Saffron=<rgba 0.9 0.885 0 0.7>
    SkyBlue=<rgba 0.0425 0.4732 0.85 0.8>
    TitaniumWhite=<rgba 1.25 1.25 1.25 0.9>
   )
   &InitializeParticle(
    Color=$Particles.Color
    Lifetime Max=&MakeCustomFloatFromBool(Bool=$User.bIsFrontendPreview, True Float=1, False Float=1)
    Lifetime Min:NiagaraFloat=&MakeCustomFloatFromBool(Bool=$User.bIsFrontendPreview, True Float=0.7, False Float=1)
    Lifetime Mode=::NewEnumerator1
    Position Offset=&AddVectorToPosition(Position=<0 0 0>, Vector=<0 0 0>)
    Sprite Rotation Mode=::NewEnumerator2
    Sprite Size Mode=::NewEnumerator1
    Uniform Sprite Size Max=40
    Uniform Sprite Size Min=30
    UsePositionOffset=false
    Write Material Random=true
   )
   &PartitionParticles(Number Of Partitions=$Emitter.PortCount, Particles Per Partition=1, Partition Mode:ENiagara_PartitionMode=::NewEnumerator1)
   &SkeletalMeshLocation(
    Bone / Socket Index=$Particles.PartitionParticles.Partition
    Bone / Socket Sampling Mode=::NewEnumerator5
    MaskSpawnGroup=false
    Mesh Sampling Type=::NewEnumerator6
    Sampled Position Offset=&AddVectorToPosition(Position=<-10 0 0>, Vector=$User.DTCarOffset)
    Skeletal Mesh=$System.SkeletalMesh
   )
   &AddVelocity(
    Cone Angle=5
    Cone Axis=&RandomRangeVector(Minimum=<-1 -0.12 0.02>, Maximum=<-1 0.12 -0.02>)
    Module.UseConeFalloff=false
    Module.UseVelDistribution=false
    Override Seed=false
    Velocity Mode=::NewEnumerator2
    Velocity Speed=&MakeCustomFloatFromBool(
     Bool=$User.bIsFrontendPreview
     True Float=&RandomRangeFloat(Minimum=1250, Maximum=1250)
     False Float=1250
    )
   )
   &ShapeLocation(MaskSpawnGroup=false, Override Seed=false, Sphere Radius=5, UseOwnerScale=true)
  ]
  particleUpdateScript [
   &ParticleState()
   &ScaleSpriteSize(
    Scale Sprite Size Mode=::NewEnumerator0
    Uniform Scale Factor=&MakeCustomFloatFromBool(
     Bool=$User.bIsPainted
     True Float=&FloatFromCurve(
      FloatCurve=DataInterface() {propertyValues {Curve {keys [
          <at 0 3 RCIM_Cubic>
          {arriveTangent=4.0478, interpMode=RCIM_Cubic, leaveTangent=7.1059, tangentMode=RCTM_Break, time=0.2, value=2.8}
          {arriveTangent=10, interpMode=RCIM_Cubic, leaveTangent=0.8726, tangentMode=RCTM_Break, time=0.5, value=5}
          <at 1 6 RCIM_Cubic>
         ]}}}
      Scale Curve=&MakeCustomFloatFromBool(Bool=$User.bIsFrontendPreview, True Float=1.25, False Float=1.5)
     )
     False Float=&FloatFromCurve(
      FloatCurve=DataInterface() {propertyValues {Curve {keys [
          <at 0 2.5 RCIM_Cubic>
          {arriveTangent=4.0478, interpMode=RCIM_Cubic, leaveTangent=7.1059, tangentMode=RCTM_Break, time=0.2, value=3}
          {arriveTangent=10, interpMode=RCIM_Cubic, leaveTangent=0.8726, tangentMode=RCTM_Break, time=0.5, value=6}
          {arriveTangent=2, interpMode=RCIM_Cubic, leaveTangent=2, time=0.8, value=6}
          <at 1 7 RCIM_Cubic>
         ]}}}
      Scale Curve=&MakeCustomFloatFromBool(Bool=$User.bIsFrontendPreview, True Float=1.15, False Float=1.15)
     )
    )
   ) {module=ScaleSpriteSize001}
   &SpriteRotationRate(Rotation Rate=&RandomRangeFloat(Minimum=125, Maximum=-125))
   &Drag(
    Drag=&FloatFromCurve(
     FloatCurve=DataInterface() {propertyValues {Curve {keys [
         {interpMode=RCIM_Cubic, tangentMode=RCTM_User, time=0.2, value=0}
         {arriveTangentWeight=0.3411, interpMode=RCIM_Cubic, tangentMode=RCTM_User, time=0.8, value=1}
        ]}}}
     Scale Curve=&MakeCustomFloatFromBool(Bool=$User.bIsFrontendPreview, True Float=4, False Float=0)
    )
    Rotational Drag=2
    Use Linear Drag=true
    Use Rotational Drag=true
   ) {enabled=false}
   &InheritSourceMovement(
    Applied Position Scale=&Multiply_VectorByFloat(Vector=<1 1 1>, Float=1)
    ApplyForce=false
    ApplyPosition=true
    ApplyPositionFalloff=false
   )
   &SolveForcesAndVelocity(Clamp Velocity=false, Limit Acceleration=false)
   &DynamicMaterialParameters(
    Index 0 Param 1=&FloatFromCurve(
     FloatCurve=DataInterface() {propertyValues {Curve {keys [
         <at 0 5 RCIM_Cubic>
         {arriveTangent=0.086, interpMode=RCIM_Cubic, leaveTangent=0.086, tangentMode=RCTM_User, time=0.5, value=1}
         <at 1 1 RCIM_Cubic>
        ]}}}
     Scale Curve=&RandomRangeFloat(Minimum=0.5, Maximum=1.5)
    )
    Index 0 Param 2=&FloatFromCurve(FloatCurve=DataInterface() {propertyValues {Curve {keys [
         <at 0 1 RCIM_Cubic>
         {arriveTangentWeight=0.2001, interpMode=RCIM_Cubic, leaveTangentWeight=0.1558, tangentMode=RCTM_Break, time=0.6, value=0.1}
         <at 1 0.5 RCIM_Cubic>
        ]}}})
    Index 0 Param 4=&FloatFromCurve(FloatCurve=DataInterface() {propertyValues {Curve {keys [
         <at 0 2 RCIM_Cubic>
         {arriveTangent=23.4207, interpMode=RCIM_Cubic, leaveTangent=16.2185, tangentMode=RCTM_Break, time=0.4, value=6}
         {arriveTangent=10, interpMode=RCIM_Cubic, leaveTangent=10, time=0.6, value=10}
         <at 1 12 RCIM_Cubic>
        ]}}})
    Module.Index0Param0Write=true
    Module.Index0Param1Write=true
    Module.Index0Param2Write=true
    Module.Index0Param3Write=true
   )
   &ScaleColor(
    Scale Alpha=&FloatFromCurve(
     FloatCurve=DataInterface() {propertyValues {Curve {keys [
         {arriveTangent=0.2974, interpMode=RCIM_Cubic, leaveTangent=0.2974, tangentMode=RCTM_User, time=0, value=0}
         {arriveTangent=-0.0024, interpMode=RCIM_Cubic, leaveTangent=0.2105, tangentMode=RCTM_Break, time=0.45, value=0.65}
         <at 0.95 0 RCIM_Cubic>
        ]}}}
     Scale Curve=&MakeCustomFloatFromBool(
      Bool=$User.bIsFrontendPreview
      True Float=&MakeCustomFloatFromBool(Bool=$User.bIsPainted, True Float=1, False Float=1.25)
      False Float=&MakeCustomFloatFromBool(Bool=$User.bIsPainted, True Float=2.5, False Float=2)
     )
    )
    Scale RGB=&VectorFromFloat(Value=&MakeCustomFloatFromBool(
      Bool=$User.bIsPainted
      True Float=&FloatFromCurve(
       FloatCurve=DataInterface() {propertyValues {Curve {keys [
           {arriveTangent=-4.1762, interpMode=RCIM_Cubic, leaveTangent=-4.1762, tangentMode=RCTM_User, time=0, value=2}
           {arriveTangent=-3.564, interpMode=RCIM_Cubic, leaveTangent=-3.564, tangentMode=RCTM_User, time=0.1, value=1}
           {arriveTangent=-1, interpMode=RCIM_Cubic, leaveTangent=-1, time=0.5, value=0.2}
           {arriveTangentWeight=0.6886, interpMode=RCIM_Cubic, tangentMode=RCTM_User, time=1, value=0.1}
          ]}}}
       Scale Curve=&MakeCustomFloatFromBool(Bool=$User.bIsFrontendPreview, True Float=2, False Float=3)
      )
      False Float=&FloatFromCurve(
       FloatCurve=DataInterface() {propertyValues {Curve {keys [<at 0 1 RCIM_Cubic>, {arriveTangent=-0.0662, interpMode=RCIM_Cubic, leaveTangent=-0.0662, tangentMode=RCTM_User, time=0.8, value=0}]}}}
       Scale Curve=&MakeCustomFloatFromBool(Bool=$User.bIsFrontendPreview, True Float=1.5, False Float=2.5)
      )
     ))
    ScaleA=true
    ScaleRGB=true
   )
  ]
  rendererClasses [&NiagaraSpriteRendererProperties]
  renderers [{rendererClass=&NiagaraSpriteRendererProperties}]
 }
}
summary {
 systemName=NS_Boost_Snow
 emitters [
  {emitterName=NE_VC_Boost_Sprite_Snow, rendererClasses [&NiagaraSpriteRendererProperties]}
  {emitterName=NE_VC_Boost_Sprite_FLAKES, simTarget=GPUComputeSim, rendererClasses [&NiagaraSpriteRendererProperties]}
  {emitterName=NE_VC_Boost_Sprite_DUST, rendererClasses [&NiagaraSpriteRendererProperties]}
 ]
 userVariables [
  User.bIsFrontendPreview {type=NiagaraBool}
  User.bIsPainted {type=NiagaraBool}
  User.ColorVariantID {type=NiagaraInt32}
  User.DTCarOffset {type=Vector3f, description="", defaultValue {struct=Vector3f, value=<1 1 1>}}
  User.FLAKESBRIGHTS {type=NiagaraFloat, defaultValue=1, description=""}
  User.PaintedColor {type=LinearColor, description="", defaultValue {struct=LinearColor, value=<rgba 1 1 1 1>}}
  User.PaintedScalar {type=NiagaraFloat, defaultValue=1, description=""}
  User.PortCount {type=NiagaraInt32, defaultValue=2}
 ]
}
